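# MF_IrisUVMask is a MetaHuman eye material function that builds a soft circular iris mask in UV space.
# It measures each UV's distance from the 0.5 center and remaps it across a limbus band ending at IrisUVRadius, with thickness set by LimbusUVWidth.
# The inverted, saturated result passes through SmoothStep, giving one inside the iris and an eased falloff to zero at the radius.
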MaterialFunction /Game/MetaHumans/Common/Face/MaterialFunctions/Eye/MF_IrisUVMask
graph {nodes [
  n1 = Length(None=Subtract(A=FunctionInput() {inputName=UV, inputType=FunctionInput_Vector2, sortPriority=1}) {constB=0.5}) {is=class MaterialExpressionLength}
  n2 = FunctionInput() {inputName=LimbusUVWidth, sortPriority=2}
  n3 = Subtract(A=FunctionInput() {inputName=IrisUVRadius, inputType=FunctionInput_Scalar}, B=@n2)
  n4 = OneMinus(None=Divide(A=Subtract(A=@n1, B=@n3), B=@n2))
  FunctionOutput(None=SmoothStep(Value=Saturate(None=@n4)))
 ]}
