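# MetaHuman eye material function that computes a refracted view direction from an internal index of refraction, a normal and a camera vector.
# It divides the refractive index of air by the internal one, then applies Snell's law in closed vector form using the normal-to-camera dot product and a square root, without trigonometric functions.
# The resulting vector is normalized and negated before output.
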MaterialFunction /Game/MetaHumans/Common/Face/MaterialFunctions/Eye/MF_RefractionDirection
graph {nodes [
  n1 = Divide(B=FunctionInput() {inputName=internalIoR, inputType=FunctionInput_Scalar}) {constA=1.0003}
  n2 = FunctionInput() {inputName=normalW, sortPriority=1}
  n3 = FunctionInput() {inputName=cameraW, sortPriority=2}
  n4 = Multiply(A=@n1, B=DotProduct(A=@n2, B=@n3)) {desc=w}
  n5 = Subtract(A=@n4, B=@n1)
  n6 = Add(A=@n1, B=@n4)
  n7 = SquareRoot(None=Add(A=Multiply(A=@n5, B=@n6)))
  n8 = Subtract(A=Multiply(A=Subtract(A=@n4, B=@n7), B=@n2), B=Multiply(A=@n1, B=@n3))
  FunctionOutput(None=Multiply(A=Normalize(VectorInput=@n8)) {constB=-1})
 ]}
